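; Movement mode extension for Umbramolt Kipper's air attack float, stopping the owner's vertical motion when it activates.
; On activation it resets the owner pawn's falling height.
; It then rewrites the character movement velocity, keeping X and Y and zeroing Z, so horizontal momentum carries through.
; The graph makes this one-time change only, with no per-tick or deactivation logic.
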(event EventOnExtensionActivate
  (params (RuntimeContext :type MMERuntime%20Context%20Structure%20%28by%20ref%29))
  (bind _ownerpawn (Variables|FortMovementMode_BaseExtLogic|GetOwnerPawn))
  (Pawn|Character|ResetFallingHeight _ownerpawn)
  (Variables|/Script/Engine.MovementComponent|SetVelocity (Math|Vector|MakeVector (.x (Variables|/Script/Engine.MovementComponent|GetVelocity (Variables|/Script/Engine.Character|GetCharacterMovement _ownerpawn))) (.y (Variables|/Script/Engine.MovementComponent|GetVelocity (Variables|/Script/Engine.Character|GetCharacterMovement _ownerpawn)))) (Variables|/Script/Engine.Character|GetCharacterMovement _ownerpawn)))
